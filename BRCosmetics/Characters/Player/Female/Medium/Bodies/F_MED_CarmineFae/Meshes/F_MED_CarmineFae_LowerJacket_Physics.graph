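# Physics asset for the lower jacket of the female medium CarmineFae player body, defining capsule bodies and joints for jacket motion.
# A spine_03 capsule anchors one back bone and two-segment front and side chains.
# Side and left-front chains hang from small base bodies with free angular joints, while right-front joins spine_03 directly.
# Jacket segments have limited swing and twist, with side chains nearly locked on one swing axis.
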PhysicsAsset /BRCosmetics/Characters/Player/Female/Medium/Bodies/F_MED_CarmineFae/Meshes/F_MED_CarmineFae_LowerJacket_Physics
bodies {
 base_lower_jacket_l [{shapeType=Capsule, length=0.505, radius=0.505, shapeName=base_lower_jacket_l_capsule, rotation=<rot 0 -90 0>}]
 base_side_lowerJacket_l_01 [{shapeType=Capsule, length=0.505, radius=0.505, shapeName=base_side_lowerJacket_l_01_capsule, rotation=<rot 0 -90 0>}]
 base_side_lowerJacket_r_01 [{shapeType=Capsule, length=0.505, radius=0.505, shapeName=base_side_lowerJacket_r_01_capsule, rotation=<rot 0 -90 0>}]
 dyn_back_jacket [{shapeType=Capsule, length=3, radius=3, center=<4 0 0.5718>, rotation=<rot 90 0 0>}]
 dyn_lower_jacket_l_01 [{shapeType=Capsule, length=3, radius=3, center=<3 0 0>, rotation=<rot 90 0 0>}]
 dyn_lower_jacket_l_02 [{shapeType=Capsule, length=3, radius=3, center=<3 0 0>, rotation=<rot 90 0 0>}]
 dyn_lower_jacket_r_01 [{shapeType=Capsule, length=3, radius=3, center=<3 0 0>, rotation=<rot 90 0 0>}]
 dyn_lower_jacket_r_02 [{shapeType=Capsule, length=3, radius=3, center=<3 0 0>, rotation=<rot 90 0 0>}]
 dyn_side_lowerJacket_l_01 [{shapeType=Capsule, length=3, radius=3, center=<4 0 0>, rotation=<rot 90 0 0>}]
 dyn_side_lowerJacket_l_02 [{shapeType=Capsule, length=3, radius=3, center=<4 0 0>, rotation=<rot 90 0 0>}]
 dyn_side_lowerJacket_r_01 [{shapeType=Capsule, length=3, radius=3, center=<4 0 0>, rotation=<rot 90 0 0>}]
 dyn_side_lowerJacket_r_02 [{shapeType=Capsule, length=3, radius=3, center=<4 0 0>, rotation=<rot 90 0 0>}]
 spine_03 [{shapeType=Capsule, length=6, radius=9, shapeName=spine_03_capsule, center=<0 3.7786 0>, rotation=<rot 90 0 0>}]
}
constraints [
 {bone1Name=dyn_back_jacket, bone2Name=spine_03, swing1LimitDegrees=45, swing2LimitDegrees=45, twistLimitDegrees=15}
 {
  bone1Name=base_lower_jacket_l
  bone2Name=spine_03
  swing1LimitDegrees=45
  swing1Motion=Free
  swing2LimitDegrees=45
  swing2Motion=Free
  twistLimitDegrees=45
  twistMotion=Free
 }
 {
  bone1Name=base_side_lowerJacket_l_01
  bone2Name=spine_03
  swing1LimitDegrees=45
  swing1Motion=Free
  swing2LimitDegrees=45
  swing2Motion=Free
  twistLimitDegrees=45
  twistMotion=Free
 }
 {bone1Name=dyn_lower_jacket_l_01, bone2Name=base_lower_jacket_l, swing1LimitDegrees=20, swing2LimitDegrees=20, twistLimitDegrees=5}
 {bone1Name=dyn_lower_jacket_l_02, bone2Name=dyn_lower_jacket_l_01, swing1LimitDegrees=20, swing2LimitDegrees=20, twistLimitDegrees=5}
 {bone1Name=dyn_side_lowerJacket_l_01, bone2Name=base_side_lowerJacket_l_01, swing1LimitDegrees=1, swing2LimitDegrees=15, twistLimitDegrees=5}
 {bone1Name=dyn_side_lowerJacket_l_02, bone2Name=dyn_side_lowerJacket_l_01, swing1LimitDegrees=1, swing2LimitDegrees=15, twistLimitDegrees=5}
 {
  bone1Name=base_side_lowerJacket_r_01
  bone2Name=spine_03
  swing1LimitDegrees=45
  swing1Motion=Free
  swing2LimitDegrees=45
  swing2Motion=Free
  twistLimitDegrees=45
  twistMotion=Free
 }
 {bone1Name=dyn_side_lowerJacket_r_01, bone2Name=base_side_lowerJacket_r_01, swing1LimitDegrees=1, swing2LimitDegrees=15, twistLimitDegrees=5}
 {bone1Name=dyn_side_lowerJacket_r_02, bone2Name=dyn_side_lowerJacket_r_01, swing1LimitDegrees=1, swing2LimitDegrees=15, twistLimitDegrees=5}
 {bone1Name=dyn_lower_jacket_r_01, bone2Name=spine_03, swing1LimitDegrees=20, swing2LimitDegrees=20, twistLimitDegrees=5}
 {bone1Name=dyn_lower_jacket_r_02, bone2Name=dyn_lower_jacket_r_01, swing1LimitDegrees=20, swing2LimitDegrees=20, twistLimitDegrees=5}
]
