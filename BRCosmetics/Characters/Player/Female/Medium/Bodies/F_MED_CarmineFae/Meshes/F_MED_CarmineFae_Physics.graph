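# Physics asset for the female medium CarmineFae player body, covering the upper torso and the jacket's chest bones, apparently for secondary jacket motion.
# Capsules on spine_04, both upper arms, two dynamic jacket bones and a right-side jacket base form the bodies.
# Upper arms and jacket base have free angular constraints to spine_04.
# Dynamic jacket bones have tight angular limits, the left constrained directly to spine_04, the right through its base.
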PhysicsAsset /BRCosmetics/Characters/Player/Female/Medium/Bodies/F_MED_CarmineFae/Meshes/F_MED_CarmineFae_Physics
bodies {
 base_Jacket_chest_r [{shapeType=Capsule, length=0.505, radius=0.505, shapeName=base_Jacket_chest_r_capsule, rotation=<rot 0 -90 0>}]
 dyn_Jacket_chest_l [{shapeType=Capsule, length=3, radius=2, center=<3 0 0>, rotation=<rot 90 0 0>}]
 dyn_Jacket_chest_r [{shapeType=Capsule, length=3, radius=4.2661, center=<3 0 0>, rotation=<rot 64.9999 90 0>}]
 spine_04 [{shapeType=Capsule, length=8, radius=8, shapeName=spine_04_capsule, center=<-0.7585 1.8676 -0.004>, rotation=<rot 2.1933 -90 -1.5425>}]
 upperarm_l [{shapeType=Capsule, length=20.602, radius=5.6088, shapeName=upperarm_r_capsule, center=<14.5545 -1.8545 0>, rotation=<rot 90 0 180>}]
 upperarm_r [{shapeType=Capsule, length=22.6184, radius=3.5923, shapeName=upperarm_r_capsule, center=<-14.5545 2.3314 0>, rotation=<rot 90 0 0>}]
}
constraints [
 {
  bone1Name=upperarm_r
  bone2Name=spine_04
  swing1LimitDegrees=45
  swing1Motion=Free
  swing2LimitDegrees=45
  swing2Motion=Free
  twistLimitDegrees=45
  twistMotion=Free
 }
 {
  bone1Name=upperarm_l
  bone2Name=spine_04
  swing1LimitDegrees=45
  swing1Motion=Free
  swing2LimitDegrees=45
  swing2Motion=Free
  twistLimitDegrees=45
  twistMotion=Free
 }
 {bone1Name=dyn_Jacket_chest_l, bone2Name=spine_04, swing1LimitDegrees=5, swing2LimitDegrees=5, twistLimitDegrees=5}
 {
  bone1Name=base_Jacket_chest_r
  bone2Name=spine_04
  swing1LimitDegrees=45
  swing1Motion=Free
  swing2LimitDegrees=45
  swing2Motion=Free
  twistLimitDegrees=45
  twistMotion=Free
 }
 {bone1Name=dyn_Jacket_chest_r, bone2Name=base_Jacket_chest_r, swing1LimitDegrees=2, swing2LimitDegrees=2, twistLimitDegrees=15}
]
