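# Post-body animation Blueprint for the CarmineFae female medium cosmetic, copying the parent mesh's pose through a Control Rig and adding corrections and secondary motion.
# Bone offsets, some gated by skydiving, hang gliding or crouching, feed two pelvis-space RigidBody passes, split by a baseball-slide jacket tweak, then zipper AnimDynamics.
# The event graph updates their shared gravity override and hang-glide flag, plus zipline and other state the AnimGraph never reads.
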AnimBlueprint /BRCosmetics/Characters/Player/Female/Medium/Bodies/F_MED_CarmineFae/Meshes/F_MED_CarmineFae_AnimBP
target_skeleton=&/F_MED_CarmineFae_Skeleton
graphs [
 AnimGraph {nodes [
   n1 = Get Gravity Override()
   n2 = RigidBody!(ComponentPose=Local To Component(LocalPose=Use cached pose 'CRCache_PreDyn'()), Alpha, OverrideWorldGravity:Vector=@n1, ExternalForce) -> (Pose) {
    bClampLinearTranslationLimitToRefPose=true
    bOverrideWorldGravity=true
    bUseExternalClothCollision=true
    overridePhysicsAsset=&/F_MED_CarmineFae_Physics
    simulationSpace=BaseBoneSpace
    alphaBoolBlend {blendOption=Linear}
    baseBoneRef {boneName=pelvis}
    componentAppliedLinearAccClamp=<3000 3000 3000>
    componentLinearAccScale=<0.25 0.25 0.25>
    componentLinearVelScale=<0.4 0.4 0.4>
    simSpaceSettings {
     maxAngularAcceleration=1000
     maxAngularVelocity=3
     maxLinearAcceleration=700
     maxLinearVelocity=500
     worldAlpha=0.6
     externalLinearDragV {x=0.2}
    }
   }
   n3 = "Transform (Modify) Bone - Bone: base_lower_jacket_l"!(ComponentPose=@n2, bAlphaBoolEnabled:bool=Get bIsBaseballSliding(), Translation="0,0,0", Rotation="0,0,50", Scale="1,1,1") -> (Pose) {
    alphaInputType=Bool
    rotationMode=BMM_Additive
    rotationSpace=BCS_BoneSpace
    translationMode=BMM_Additive
    translationSpace=BCS_BoneSpace
    alphaBoolBlend {blendInTime=0.25, blendOption=Linear, blendOutTime=0.25}
    boneToModify {boneName=base_lower_jacket_l}
    rotation {roll=50}
   }
   n4 = RigidBody!(
    ComponentPose=@n3
    Alpha=0.7
    OverrideWorldGravity=@n1
    SimSpaceSettings:Sim Space Settings Structure=(WorldAlpha=0.8,VelocityScaleZ=1,DampingAlpha=1,MaxLinearVelocity=500,MaxAngularVelocity=3,MaxLinearAcceleration=700,MaxAngularAcceleration=1000,ExternalLinearDragV=(X=0.2,Y=0.2,Z=0.1),ExternalLinearVelocity=(X=0,Y=0,Z=0),ExternalAngularVelocity=(X=0,Y=0,Z=0))
   ) -> (Pose) {
    bOverrideWorldGravity=true
    overridePhysicsAsset=&/F_MED_CarmineFae_LowerJacket_Physics
    simulationSpace=BaseBoneSpace
    alphaBoolBlend {blendOption=Linear}
    baseBoneRef {boneName=pelvis}
    componentAppliedLinearAccClamp=<3000 3000 1000>
    componentLinearAccScale=<0.25 0.25 0.1>
    componentLinearVelScale=<0.25 0.25 0.1>
    overrideWorldGravity {z=-980}
    simSpaceSettings {
     maxAngularAcceleration=1000
     maxAngularVelocity=3
     maxLinearAcceleration=700
     maxLinearVelocity=500
     worldAlpha=0.8
     externalLinearDragV=<0.2 0.2 0.1>
    }
   }
   n5 = Anim Dynamics - Bone: dyn_jacket_zipper!(ComponentPose=@n4, Alpha=1, GravityOverride:Vector=@n1, ExternalForce="0,0,0") -> (Pose) {
    angularDampingOverride=1
    angularSpringConstant=20000
    bAngularSpring=true
    bOverrideAngularDamping=true
    bUseGravityOverride=true
    gravityScale=0.35
    simulationSpace=RootRelative
    alphaBoolBlend {blendOption=Linear}
    boundBone {boneName=dyn_jacket_zipper}
    componentLinearAccScale=<1 1 1>
    componentLinearVelScale=<1 1 1>
    gravityOverride {z=-980}
    physicsBodyDefinitions [{
      boundBone {boneName=dyn_jacket_zipper}
      boxExtents=<20 30 30>
      constraintSetup {angularLimitsMax=<15 15 15>, angularLimitsMin=<-15 -15 -15>, angularTarget=<1 0 0>}
      localJointOffset=<4 0 0>
     }]
   }
   Output Pose(Result=Component To Local(ComponentPose=@n5)) {layerGroup=None}
   n6 = "Transform (Modify) Bone - Bone: attach_backpack"!(
    ComponentPose=Local To Component(LocalPose=Use cached pose 'CRCache_Input'())
    bAlphaBoolEnabled:bool=Get bIsSkyDiving()
    Translation="0,-2,0"
    Rotation="0,0,0"
    Scale="1,1,1"
   ) -> (Pose) {
    alphaInputType=Bool
    rotationMode=BMM_Additive
    rotationSpace=BCS_BoneSpace
    translationMode=BMM_Additive
    translationSpace=BCS_BoneSpace
    alphaBoolBlend {blendInTime=0.25, blendOption=Linear, blendOutTime=0.25}
    boneToModify {boneName=attach_backpack}
    translation {y=-2}
   }
   n7 = Get HangGliding() -> (HangGliding:bool)
   n8 = "Transform (Modify) Bone - Bone: deform_upperarm_up_base_l"!(ComponentPose=@n6, bAlphaBoolEnabled:bool=@n7, Translation="-5,0,-20", Rotation="0,0,0", Scale="1,1,1") -> (Pose) {
    alphaInputType=Bool
    translationMode=BMM_Additive
    translationSpace=BCS_BoneSpace
    alphaBoolBlend {blendInTime=0.25, blendOption=Linear, blendOutTime=0.25}
    boneToModify {boneName=deform_upperarm_up_base_l}
    translation {x=-5, z=-20}
   }
   n9 = "Transform (Modify) Bone - Bone: deform_upperarm_up_r"!(ComponentPose=@n8, bAlphaBoolEnabled:bool=@n7, Translation="-5,0,20", Rotation="0,0,0", Scale="1,1,1") -> (Pose) {
    alphaInputType=Bool
    translationMode=BMM_Additive
    translationSpace=BCS_BoneSpace
    alphaBoolBlend {blendInTime=0.25, blendOption=Linear, blendOutTime=0.25}
    boneToModify {boneName=deform_upperarm_up_r}
    translation {x=-5, z=20}
   }
   n10 = "Transform (Modify) Bone - Bone: jnt_helper_inner_leg_l"(ComponentPose=@n9, Alpha=1, Translation="-0.033,-4.314,-6", Rotation="0,0,0", Scale="1,1,1") {
    translationMode=BMM_Additive
    translationSpace=BCS_BoneSpace
    alphaBoolBlend {blendOption=Linear}
    boneToModify {boneName=jnt_helper_inner_leg_l}
    translation=<-0.033 -4.314 -6>
   }
   n11 = "Transform (Modify) Bone - Bone: jnt_helper_inner_reg_r"(ComponentPose=@n10, Alpha=1, Translation="1.623,2.181,6.347", Rotation="0,0,0", Scale="1,1,1") {
    translationMode=BMM_Additive
    translationSpace=BCS_BoneSpace
    alphaBoolBlend {blendOption=Linear}
    boneToModify {boneName=jnt_helper_inner_reg_r}
    translation=<1.623 2.181 6.347>
   }
   n12 = "Transform (Modify) Bone - Bone: jnt_helper_inner_leg_b_l"(ComponentPose=@n11, Alpha=1, Translation="-0.000,3.564,-0.000", Rotation="0,0,0", Scale="1,1,1") {
    translationMode=BMM_Additive
    translationSpace=BCS_BoneSpace
    alphaBoolBlend {blendOption=Linear}
    boneToModify {boneName=jnt_helper_inner_leg_b_l}
    translation {y=3.564}
   }
   n13 = "Transform (Modify) Bone - Bone: jnt_helper_inner_reg_b_r"(ComponentPose=@n12, Alpha=1, Translation="-0.522,5.343,0.082", Rotation="0,0,0", Scale="1,1,1") {
    translationMode=BMM_Additive
    translationSpace=BCS_BoneSpace
    alphaBoolBlend {blendOption=Linear}
    boneToModify {boneName=jnt_helper_inner_reg_b_r}
    translation=<-0.522 5.343 0.082>
   }
   n14 = "Transform (Modify) Bone - Bone: Back_skirt_Helper_r"!(ComponentPose=@n13, bAlphaBoolEnabled:bool=Get bIsCrouching(), Translation="0,0,0", Rotation="-0.232,7.229,-27.876", Scale="1,1,1") -> (Pose) {
    alphaInputType=Bool
    rotationMode=BMM_Additive
    rotationSpace=BCS_BoneSpace
    translationMode=BMM_Additive
    alphaBoolBlend {blendInTime=0.25, blendOption=Linear, blendOutTime=0.25}
    boneToModify {boneName=Back_skirt_Helper_r}
    rotation=<rot -0.232 7.229 -27.8761>
   }
   n15 = "Transform (Modify) Bone - Bone: Base_Back_skirt_Helper_l"!(ComponentPose=@n14, bAlphaBoolEnabled:bool=True, Translation="2,-1,0", Rotation="0,0,0", Scale="1,1,1") -> (Pose) {
    alphaInputType=Bool
    translationMode=BMM_Additive
    translationSpace=BCS_BoneSpace
    alphaBoolBlend {blendInTime=0.1, blendOption=Linear, blendOutTime=0.1}
    boneToModify {boneName=Base_Back_skirt_Helper_l}
    translation=<2 -1>
   }
   n16 = "Transform (Modify) Bone - Bone: Base_Back_skirt_Helper_r"!(ComponentPose=@n15, bAlphaBoolEnabled:bool=True, Translation="-1,-1,0", Rotation="0,0,0", Scale="1,1,1") -> (Pose) {
    alphaInputType=Bool
    translationMode=BMM_Additive
    translationSpace=BCS_BoneSpace
    alphaBoolBlend {blendInTime=0.1, blendOption=Linear, blendOutTime=0.1}
    boneToModify {boneName=Base_Back_skirt_Helper_r}
    translation=<-1 -1>
   }
   n17 = "Transform (Modify) Bone - Bone: twist_helper"(ComponentPose=@n16, Alpha=1, Translation="0,0,0", Rotation="0,0,0", Scale="1,0.5,0.5") {
    scaleMode=BMM_Additive
    scaleSpace=BCS_BoneSpace
    alphaBoolBlend {blendOption=Linear}
    boneToModify {boneName=twist_helper}
    scale {y=0.5, z=0.5}
   }
   Save cached pose 'CRCache_PreDyn'(Pose=Component To Local(ComponentPose=@n17)) {cacheName=CRCache_PreDyn}
   n18 = Control Rig(Source=Copy Pose From Mesh() {bUseAttachedParent=true}) {
    alphaBoolBlend {blendOption=Linear}
    controlRigAssetReference {blueprintRigClass=&/F_MED_CarmineFae_CtrlRig.F_MED_CarmineFae_CtrlRig_C}
   }
   Save cached pose 'CRCache_Input'(Pose=@n18) {cacheName=CRCache_Input}
  ]}
 EventGraph {nodes [
   Event BlueprintUpdateAnimation()
   > n1 = Gravity Override State Function DEPRECATED(
    CharacterPartAnimInstance=Self-Reference()
    GameTimeInSeconds=Get Game Time in Seconds() -> (Game Time in Seconds:double)
    GravityOverrideParams=Get Gravity Override Params()
   )
   > Set Gravity Override(Gravity Override=@n1)
   Event OnParachuteOpened()
   > Switch on EGliderType(Selection=Get PlayerGliderType() -> (PlayerGliderType=HangGlider))
   > Set HangGliding(HangGliding:bool=true) -> (Output_Get:bool)
   Event OnBeginSkydiving()
   > Set Skydiving(Skydiving:bool=true) -> (Output_Get:bool)
   Event OnEndSkydiving()
   > Set Skydiving(Skydiving=false)
   Event OnParachuteClosed()
   > Set HangGliding(HangGliding=false)
   Event OnBeginCrouching()
   > Set Crouching(Crouching:bool=false) -> (Output_Get:bool)
   Event OnEndCrouching()
   > Set Crouching(Crouching=false)
   n2 = Event BlueprintProcessAnimInputEvent()
   n3 = Branch(Condition=IsNamed(self=@n2.AnimInputEvent, InEventName=aie.Player.Zipline.Entered))
   > Set OnZipline(OnZipline:bool=1) -> (Output_Get:bool)
   > Set ZiplineAlpha(ZiplineAlpha:double=0) -> (Output_Get:double)
   Branch(
    execute=@n3.else
    Condition=OR Boolean!(A=IsNamed(self=@n2.AnimInputEvent, InEventName=aie.Player.Zipline.Exited), B=Get bIsMoving2D(), C:bool=Get bIsJumping()) -> (ReturnValue)
   )
   > Set OnZipline(OnZipline=false)
   > Set ZiplineAlpha(ZiplineAlpha=1)
   n4 = GetOwnerSkeletalMeshComponent()
   n5 = BreakRotator(InRot=BreakTransform(InTransform=GetDeltaTransformFromRefPose(self=@n4, BoneName=spine_05, BaseName=spine_01)).Rotation)
   Set backpack_translation(backpack_translation=MakeVector(Y=MapRangeClamped(Value=@n5.Yaw, InRangeA=0, InRangeB=-30, OutRangeB=-5)))
   n6 = MapRangeClamped(Value=@n5.Yaw, InRangeA=0, InRangeB=-30, OutRangeB=-40)
   > Set backpack_rotate(backpack_rotate=MakeRotator(Roll=@n6, Pitch=@n6, Yaw=@n6))
   n7 = Get bIsFalling()
   n8 = Get bIsJumping()
   Set TacticalJump(TacticalJump:bool=AND Boolean(A=Get bIsTacticalSprinting(), B=OR Boolean!(A=@n7, B=@n8) -> (ReturnValue))) -> (Output_Get:bool)
  ]}
]
variables [
 backpack_rotate {type=Rotator}
 backpack_translation {type=Vector}
 Crouching {type=boolean}
 Game Time in Seconds {type=number}
 Gravity Override {type=Vector}
 Gravity Override Params {type=GravityOverrideParamsStruct}
 HangGliding {type=boolean}
 OnZipline {type=boolean}
 Skydiving {type=boolean}
 TacticalJump {type=boolean}
 ZiplineAlpha {type=number}
]
