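# Physics asset for the CarmineFae female medium face accessory mesh, covering left and right long hair strands.
# Each strand is a three-capsule dyn chain hanging from a base capsule on the head, with per-joint swing and twist limits, mostly tighter on the right.
# Capsules on head, neck, upper spine, both clavicles and only the right upper arm form the body, joined by free angular constraints.
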PhysicsAsset /BRCosmetics/Characters/Player/Female/Medium/Bodies/F_MED_CarmineFae/Meshes/Parts/F_MED_CarmineFae_FaceAcc_Physics
bodies {
 base_long_hair_l [{shapeType=Capsule, length=0.505, radius=0.505, shapeName=base_long_hair_l_capsule, rotation=<rot 0 -90 0>}]
 base_long_hair_r [{shapeType=Capsule, length=0.505, radius=0.505, shapeName=base_long_hair_r_capsule, rotation=<rot 0 -90 0>}]
 clavicle_l [{shapeType=Capsule, length=5, radius=5, shapeName=clavicle_r_capsule, center=<5 0 0>, rotation=<rot 90 180 0>}]
 clavicle_r [{shapeType=Capsule, length=5, radius=5, shapeName=clavicle_r_capsule, center=<-5 0 0>, rotation=<rot 90 0 0>}]
 dyn_long_hair_l_01 [{shapeType=Capsule, length=1, radius=1, center=<1 0 0>, rotation=<rot 90 0 0>}]
 dyn_long_hair_l_02 [{shapeType=Capsule, length=1, radius=1, center=<1 0 0>, rotation=<rot 90 0 0>}]
 dyn_long_hair_l_03 [{shapeType=Capsule, length=1, radius=1, center=<1 0 0>, rotation=<rot 90 0 0>}]
 dyn_long_hair_r_01 [{shapeType=Capsule, length=1, radius=1, center=<1 0 0>, rotation=<rot 90 0 0>}]
 dyn_long_hair_r_02 [{shapeType=Capsule, length=1, radius=1, center=<1 0 0>, rotation=<rot 90 0 0>}]
 dyn_long_hair_r_03 [{shapeType=Capsule, length=1, radius=1, center=<1 0 0>, rotation=<rot 90 0 0>}]
 head [{shapeType=Capsule, length=8, radius=4.2767, center=<-11.888 -1.8418 0>, rotation=<rot 90 0 0>}]
 neck_01 [{shapeType=Capsule, length=4.5, radius=6.0707, rotation=<rot 90 0 0>}]
 spine_04 [{shapeType=Capsule, length=23.5, radius=5.5, shapeName=spine_04_capsule, center=<7.5578 7.4509 0>, rotation=<rot 0 -90 0>}]
 spine_05 [{shapeType=Capsule, length=30.56, radius=5.3835, shapeName=spine_05_capsule, center=<5.9342 0.6201 0>}]
 upperarm_r [{shapeType=Capsule, length=18, radius=4, shapeName=upperarm_r_capsule, center=<-13.2736 0.8903 0>, rotation=<rot 90 0 0>}]
}
constraints [
 {
  bone1Name=base_long_hair_l
  bone2Name=head
  swing1LimitDegrees=45
  swing1Motion=Free
  swing2LimitDegrees=45
  swing2Motion=Free
  twistLimitDegrees=45
  twistMotion=Free
 }
 {bone1Name=dyn_long_hair_l_01, bone2Name=base_long_hair_l, swing1LimitDegrees=18, swing2LimitDegrees=18, twistLimitDegrees=5}
 {bone1Name=dyn_long_hair_l_02, bone2Name=dyn_long_hair_l_01, swing1LimitDegrees=15, swing2LimitDegrees=30, twistLimitDegrees=5}
 {bone1Name=dyn_long_hair_l_03, bone2Name=dyn_long_hair_l_02, swing1LimitDegrees=35, swing2LimitDegrees=35, twistLimitDegrees=10}
 {
  bone1Name=head
  bone2Name=neck_01
  swing1LimitDegrees=45
  swing1Motion=Free
  swing2LimitDegrees=45
  swing2Motion=Free
  twistLimitDegrees=45
  twistMotion=Free
 }
 {
  bone1Name=neck_01
  bone2Name=spine_05
  swing1LimitDegrees=45
  swing1Motion=Free
  swing2LimitDegrees=45
  swing2Motion=Free
  twistLimitDegrees=45
  twistMotion=Free
 }
 {
  bone1Name=base_long_hair_r
  bone2Name=head
  swing1LimitDegrees=45
  swing1Motion=Free
  swing2LimitDegrees=45
  swing2Motion=Free
  twistLimitDegrees=45
  twistMotion=Free
 }
 {bone1Name=dyn_long_hair_r_01, bone2Name=base_long_hair_r, swing1LimitDegrees=15, swing2LimitDegrees=15, twistLimitDegrees=5}
 {bone1Name=dyn_long_hair_r_02, bone2Name=dyn_long_hair_r_01, swing1LimitDegrees=8, swing2LimitDegrees=15, twistLimitDegrees=5}
 {bone1Name=dyn_long_hair_r_03, bone2Name=dyn_long_hair_r_02, swing1LimitDegrees=25, swing2LimitDegrees=25, twistLimitDegrees=5}
 {
  bone1Name=clavicle_r
  bone2Name=spine_05
  swing1LimitDegrees=45
  swing1Motion=Free
  swing2LimitDegrees=45
  swing2Motion=Free
  twistLimitDegrees=45
  twistMotion=Free
 }
 {
  bone1Name=clavicle_l
  bone2Name=spine_05
  swing1LimitDegrees=45
  swing1Motion=Free
  swing2LimitDegrees=45
  swing2Motion=Free
  twistLimitDegrees=45
  twistMotion=Free
 }
 {
  bone1Name=upperarm_r
  bone2Name=clavicle_r
  swing1LimitDegrees=45
  swing1Motion=Free
  swing2LimitDegrees=45
  swing2Motion=Free
  twistLimitDegrees=45
  twistMotion=Free
 }
 {
  bone1Name=spine_05
  bone2Name=spine_04
  swing1LimitDegrees=45
  swing1Motion=Free
  swing2LimitDegrees=45
  swing2Motion=Free
  twistLimitDegrees=45
  twistMotion=Free
 }
]
