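# Post-process style AnimBP for Carmine Fae's face accessory mesh, animating hair and a medallion on the parent's copied pose.
# Control Rig feeds chained AnimDynamics hair, weighted multi-bone long-hair constraints, then medallion constraint and dynamics with front-end-only adjustments, staged through cached poses.
# A final RigidBody pass simulates in spine_05 space, using gravity updated each frame by a deprecated helper.
# Glider events toggle HangGliding, blending in a left long-hair rotation offset.
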AnimBlueprint /BRCosmetics/Characters/Player/Female/Medium/Bodies/F_MED_CarmineFae/Meshes/Parts/F_MED_CarmineFae_FaceAcc_AnimBP
target_skeleton=&/F_MED_CarmineFae_FaceAcc_Skeleton
graphs [
 AnimGraph {nodes [
   n1 = Get HangGliding() -> (HangGliding:bool)
   n2 = "Transform (Modify) Bone - Bone: base_long_hair_l"!(
    ComponentPose=Local To Component(LocalPose=Use cached pose 'MedallionCache'())
    bAlphaBoolEnabled:bool=@n1
    Translation="0,0,0"
    Rotation="0,0,-8"
    Scale="1,1,1"
   ) -> (Pose) {
    alphaInputType=Bool
    rotationMode=BMM_Additive
    rotationSpace=BCS_BoneSpace
    alphaBoolBlend {blendInTime=0.25, blendOption=Linear, blendOutTime=0.25}
    boneToModify {boneName=base_long_hair_l}
    rotation {roll=-8}
   }
   n3 = "Transform (Modify) Bone - Bone: base_long_hair_r"!(ComponentPose=@n2, bAlphaBoolEnabled:bool=@n1, Translation="0,0,0", Rotation="0,0,0", Scale="1,1,1") -> (Pose) {
    alphaInputType=Bool
    rotationMode=BMM_Additive
    rotationSpace=BCS_BoneSpace
    alphaBoolBlend {blendInTime=0.25, blendOption=Linear, blendOutTime=0.25}
    boneToModify {boneName=base_long_hair_r}
   }
   n4 = RigidBody!(ComponentPose=@n3, Alpha, OverrideWorldGravity:Vector=Get Gravity Override(), ExternalForce) -> (Pose) {
    bOverrideWorldGravity=true
    overridePhysicsAsset=&/F_MED_CarmineFae_FaceAcc_Physics
    simulationSpace=BaseBoneSpace
    alphaBoolBlend {blendOption=Linear}
    baseBoneRef {boneName=spine_05}
    componentAppliedLinearAccClamp=<3000 3000 3000>
    componentLinearAccScale=<0.4 0.3 0.3>
    componentLinearVelScale=<0.4 0.3 0.3>
    simSpaceSettings {maxAngularAcceleration=1000, maxAngularVelocity=3, maxLinearAcceleration=250, maxLinearVelocity=500, velocityScaleZ=0.45, worldAlpha=0.6}
   }
   Output Pose(Result=Component To Local(ComponentPose=@n4)) {layerGroup=None}
   n5 = Control Rig(Source=Copy Pose From Mesh() {bUseAttachedParent=true}) {
    alphaBoolBlend {blendOption=Linear}
    controlRigAssetReference {blueprintRigClass=&/F_MED_CarmineFae_FaceAcc_CtrlRig.F_MED_CarmineFae_FaceAcc_CtrlRig_C}
   }
   Save cached pose 'Cache01'(Pose=@n5) {cacheName=Cache01}
   n6 = Anim Dynamics - Bone: dyn_top_hair(ComponentPose=Local To Component(LocalPose=Use cached pose 'Cache01'()), Alpha=0.7, ExternalForce="0,0,0") {
    angularDampingOverride=1
    angularSpringConstant=20000
    bAngularSpring=true
    bOverrideAngularDamping=true
    bUseGravityOverride=true
    gravityScale=0.35
    simulationSpace=RootRelative
    alphaBoolBlend {blendOption=Linear}
    boundBone {boneName=dyn_top_hair}
    componentLinearAccScale=<0.8 0.8 0.8>
    componentLinearVelScale=<0.8 0.8 0.8>
    gravityOverride {z=-980}
    physicsBodyDefinitions [{
      boundBone {boneName=dyn_top_hair}
      boxExtents=<20 30 30>
      constraintSetup {angularLimitsMax=<25 25 25>, angularLimitsMin=<-25 -25 -45>, angularTarget=<1 0 0>}
      localJointOffset=<4 0 0>
     }]
   }
   n7 = Anim Dynamics - Bone: dyn_top_hair_01(ComponentPose=@n6, Alpha=0.8, ExternalForce="0,0,0") {
    angularDampingOverride=1
    angularSpringConstant=20000
    bAngularSpring=true
    bOverrideAngularDamping=true
    bUseGravityOverride=true
    gravityScale=0.35
    simulationSpace=RootRelative
    alphaBoolBlend {blendOption=Linear}
    boundBone {boneName=dyn_top_hair_01}
    componentLinearAccScale=<0.8 0.8 0.8>
    componentLinearVelScale=<0.8 0.8 0.8>
    gravityOverride {z=-980}
    physicsBodyDefinitions [{
      boundBone {boneName=dyn_top_hair_01}
      boxExtents=<20 30 30>
      constraintSetup {angularLimitsMax=<45 45 45>, angularLimitsMin=<-45 -45 -45>, angularTarget=<1 0 0>}
      localJointOffset=<4 0 0>
     }]
   }
   n8 = Anim Dynamics - Bone: dyn_top_hair_02(ComponentPose=@n7, Alpha=0.8, ExternalForce="0,0,0") {
    angularDampingOverride=1
    angularSpringConstant=20000
    bAngularSpring=true
    bOverrideAngularDamping=true
    bUseGravityOverride=true
    gravityScale=0.35
    simulationSpace=RootRelative
    alphaBoolBlend {blendOption=Linear}
    boundBone {boneName=dyn_top_hair_02}
    componentLinearAccScale=<0.6 0.6 0.6>
    componentLinearVelScale=<0.6 0.6 0.6>
    gravityOverride {z=-980}
    physicsBodyDefinitions [{
      boundBone {boneName=dyn_top_hair_02}
      boxExtents=<20 30 30>
      constraintSetup {angularLimitsMax=<45 45 45>, angularLimitsMin=<-45 -45 -45>, angularTarget=<1 0 0>}
      localJointOffset=<3 0 0>
     }]
   }
   n9 = Anim Dynamics - Bone: dyb_hair_a_01(ComponentPose=@n8, Alpha=0.7, ExternalForce="0,0,0") {
    angularDampingOverride=1
    angularSpringConstant=20000
    bAngularSpring=true
    bOverrideAngularDamping=true
    bUseGravityOverride=true
    gravityScale=0.35
    simulationSpace=RootRelative
    alphaBoolBlend {blendOption=Linear}
    boundBone {boneName=dyb_hair_a_01}
    componentLinearAccScale=<0.6 0.6 0.6>
    componentLinearVelScale=<0.6 1 0.6>
    gravityOverride {z=-980}
    physicsBodyDefinitions [{
      boundBone {boneName=dyb_hair_a_01}
      boxExtents=<20 30 30>
      constraintSetup {angularLimitsMax=<30 30 30>, angularLimitsMin=<-30 -30 -30>, angularTarget=<1 0 0>}
      localJointOffset=<3 0 0>
     }]
   }
   n10 = Anim Dynamics - Bone: dyb_hair_a_02(ComponentPose=@n9, Alpha=0.7, ExternalForce="0,0,0") {
    angularDampingOverride=1
    angularSpringConstant=20000
    bAngularSpring=true
    bOverrideAngularDamping=true
    bUseGravityOverride=true
    gravityScale=0.35
    simulationSpace=RootRelative
    alphaBoolBlend {blendOption=Linear}
    boundBone {boneName=dyb_hair_a_02}
    componentLinearAccScale=<0.6 0.6 0.6>
    componentLinearVelScale=<0.6 1 0.6>
    gravityOverride {z=-980}
    physicsBodyDefinitions [{
      boundBone {boneName=dyb_hair_a_02}
      boxExtents=<20 30 30>
      constraintSetup {angularLimitsMax=<45 45 45>, angularLimitsMin=<-45 -45 -45>, angularTarget=<1 0 0>}
      localJointOffset=<3 0 0>
     }]
   }
   n11 = Anim Dynamics - Bone: dyn_hair_b_01(ComponentPose=@n10, Alpha=0.8, ExternalForce="0,0,0") {
    angularDampingOverride=1
    angularSpringConstant=20000
    bAngularSpring=true
    bOverrideAngularDamping=true
    bUseGravityOverride=true
    gravityScale=0.35
    simulationSpace=RootRelative
    alphaBoolBlend {blendOption=Linear}
    boundBone {boneName=dyn_hair_b_01}
    componentLinearAccScale=<0.6 0.6 0.6>
    componentLinearVelScale=<0.6 0.6 0.6>
    gravityOverride {z=-980}
    physicsBodyDefinitions [{
      boundBone {boneName=dyn_hair_b_01}
      boxExtents=<20 30 30>
      constraintSetup {angularLimitsMax=<30 30 30>, angularLimitsMin=<-30 -30 -30>, angularTarget=<1 0 0>}
      localJointOffset=<3 0 0>
     }]
   }
   n12 = Anim Dynamics - Bone: dyn_hair_b_02(ComponentPose=@n11, Alpha=0.9, ExternalForce="0,0,0") {
    angularDampingOverride=1
    angularSpringConstant=20000
    bAngularSpring=true
    bOverrideAngularDamping=true
    bUseGravityOverride=true
    gravityScale=0.35
    simulationSpace=RootRelative
    alphaBoolBlend {blendOption=Linear}
    boundBone {boneName=dyn_hair_b_02}
    componentLinearAccScale=<0.6 0.6 0.6>
    componentLinearVelScale=<0.6 0.6 0.6>
    gravityOverride {z=-980}
    physicsBodyDefinitions [{
      boundBone {boneName=dyn_hair_b_02}
      boxExtents=<20 30 30>
      constraintSetup {angularLimitsMax=<45 45 45>, angularLimitsMin=<-45 -45 -45>, angularTarget=<1 0 0>}
      localJointOffset=<3 0 0>
     }]
   }
   n13 = Anim Dynamics - Bone: dyn_hair_c_01(ComponentPose=@n12, Alpha=0.8, ExternalForce="0.8,0,0") {
    angularDampingOverride=1
    angularSpringConstant=20000
    bAngularSpring=true
    bOverrideAngularDamping=true
    bUseGravityOverride=true
    gravityScale=0.35
    simulationSpace=RootRelative
    alphaBoolBlend {blendOption=Linear}
    boundBone {boneName=dyn_hair_c_01}
    componentLinearAccScale=<0.6 0.6 0.6>
    componentLinearVelScale=<0.6 0.6 0.6>
    gravityOverride {z=-980}
    physicsBodyDefinitions [{
      boundBone {boneName=dyn_hair_c_01}
      boxExtents=<20 30 30>
      constraintSetup {angularLimitsMax=<30 30 30>, angularLimitsMin=<-30 -30 -30>, angularTarget=<1 0 0>}
      localJointOffset=<3 0 0>
     }]
   }
   n14 = Anim Dynamics - Bone: dyn_hair_c_02(ComponentPose=@n13, Alpha=0.9, ExternalForce="0,0,0") {
    angularDampingOverride=1
    angularSpringConstant=20000
    bAngularSpring=true
    bOverrideAngularDamping=true
    bUseGravityOverride=true
    gravityScale=0.35
    simulationSpace=RootRelative
    alphaBoolBlend {blendOption=Linear}
    boundBone {boneName=dyn_hair_c_02}
    componentLinearAccScale=<0.6 0.6 0.6>
    componentLinearVelScale=<0.6 0.6 0.6>
    gravityOverride {z=-980}
    physicsBodyDefinitions [{
      boundBone {boneName=dyn_hair_c_02}
      boxExtents=<20 30 30>
      constraintSetup {angularLimitsMax=<15 15 15>, angularLimitsMin=<-45 -45 -45>, angularTarget=<1 0 0>}
      localJointOffset=<3 0 0>
     }]
   }
   n15 = Anim Dynamics - Bone: dyn_hair_d_01(ComponentPose=@n14, Alpha=0.8, ExternalForce="0,0,0") {
    angularDampingOverride=1
    angularSpringConstant=20000
    bAngularSpring=true
    bOverrideAngularDamping=true
    bUseGravityOverride=true
    gravityScale=0.35
    simulationSpace=RootRelative
    alphaBoolBlend {blendOption=Linear}
    boundBone {boneName=dyn_hair_d_01}
    componentLinearAccScale=<0.6 0.6 0.6>
    componentLinearVelScale=<0.6 0.6 0.6>
    gravityOverride {z=-980}
    physicsBodyDefinitions [{
      boundBone {boneName=dyn_hair_d_01}
      boxExtents=<20 30 30>
      constraintSetup {angularLimitsMax=<30 30 30>, angularLimitsMin=<-30 -30 -30>, angularTarget=<1 0 0>}
      localJointOffset=<3 0 0>
     }]
   }
   n16 = Anim Dynamics - Bone: dyn_hair_d_02(ComponentPose=@n15, Alpha=0.9, ExternalForce="0,0,0") {
    angularDampingOverride=1
    angularSpringConstant=20000
    bAngularSpring=true
    bOverrideAngularDamping=true
    bUseGravityOverride=true
    gravityScale=0.35
    simulationSpace=RootRelative
    alphaBoolBlend {blendOption=Linear}
    boundBone {boneName=dyn_hair_d_02}
    componentLinearAccScale=<0.6 0.6 0.6>
    componentLinearVelScale=<0.6 0.6 0.6>
    gravityOverride {z=-980}
    physicsBodyDefinitions [{
      boundBone {boneName=dyn_hair_d_02}
      boxExtents=<20 30 30>
      constraintSetup {angularLimitsMax=<45 45 45>, angularLimitsMin=<-45 -45 -45>, angularTarget=<1 0 0>}
      localJointOffset=<3 0 0>
     }]
   }
   n17 = Anim Dynamics - Bone: dyn_hair_e_01(ComponentPose=@n16, Alpha=0.8, ExternalForce="0,0,0") {
    angularDampingOverride=1
    angularSpringConstant=20000
    bAngularSpring=true
    bOverrideAngularDamping=true
    bUseGravityOverride=true
    gravityScale=0.35
    simulationSpace=RootRelative
    alphaBoolBlend {blendOption=Linear}
    boundBone {boneName=dyn_hair_e_01}
    componentLinearAccScale=<0.6 0.6 0.6>
    componentLinearVelScale=<0.6 0.6 0.6>
    gravityOverride {z=-980}
    physicsBodyDefinitions [{
      boundBone {boneName=dyn_hair_e_01}
      boxExtents=<20 30 30>
      constraintSetup {angularLimitsMax=<15 15 15>, angularLimitsMin=<-15 -15 -15>, angularTarget=<1 0 0>}
      localJointOffset=<3 0 0>
     }]
   }
   n18 = Anim Dynamics - Bone: dyn_hair_e_02(ComponentPose=@n17, Alpha=0.9, ExternalForce="0,0,0") {
    angularDampingOverride=1
    angularSpringConstant=20000
    bAngularSpring=true
    bOverrideAngularDamping=true
    bUseGravityOverride=true
    gravityScale=0.35
    simulationSpace=RootRelative
    alphaBoolBlend {blendOption=Linear}
    boundBone {boneName=dyn_hair_e_02}
    componentLinearAccScale=<0.6 0.6 0.6>
    componentLinearVelScale=<0.6 0.6 0.6>
    gravityOverride {z=-980}
    physicsBodyDefinitions [{
      boundBone {boneName=dyn_hair_e_02}
      boxExtents=<20 30 30>
      constraintSetup {angularLimitsMax=<45 45 45>, angularLimitsMin=<-45 -45 -45>, angularTarget=<1 0 0>}
      localJointOffset=<3 0 0>
     }]
   }
   n19 = "Constraint - base_long_hair_l (PPPPPPPP)"!(
    ComponentPose=@n18
    Alpha=1
    ConstraintWeights_0=0.4675
    ConstraintWeights_1:float=0.2424
    ConstraintWeights_2:float=0.1397
    ConstraintWeights_3:float=0.0766
    ConstraintWeights_4:float=0.0541
    ConstraintWeights_5:float=0.0183
    ConstraintWeights_6:float=0.001
    ConstraintWeights_7:float=0.0004
   ) -> (Pose) {
    alphaBoolBlend {blendOption=Linear}
    boneToModify {boneName=base_long_hair_l}
    constraintSetup [
     {targetBone {boneName=spine_05}}
     {targetBone {boneName=neck_02}}
     {targetBone {boneName=neck_01}}
     {targetBone {boneName=head}}
     {targetBone {boneName=clavicle_l}}
     {targetBone {boneName=spine_03}}
     {targetBone {boneName=clavicle_r}}
     {targetBone {boneName=spine_04}}
    ]
    constraintWeights [1,1,1,1,1,1,1,1]
   }
   n20 = "Constraint - dyn_long_hair_l_02 (PPPPPPPP)"!(
    ComponentPose=@n19
    Alpha=1
    ConstraintWeights_0=0.5095
    ConstraintWeights_1:float=0.3403
    ConstraintWeights_2:float=0.0468
    ConstraintWeights_3:float=0.0453
    ConstraintWeights_4:float=0.0315
    ConstraintWeights_5:float=0.0122
    ConstraintWeights_6:float=0.0097
    ConstraintWeights_7:float=0.0047
   ) -> (Pose) {
    alphaBoolBlend {blendOption=Linear}
    boneToModify {boneName=dyn_long_hair_l_02}
    constraintSetup [
     {targetBone {boneName=spine_05}}
     {targetBone {boneName=clavicle_l}}
     {targetBone {boneName=neck_01}}
     {targetBone {boneName=neck_02}}
     {targetBone {boneName=spine_04}}
     {targetBone {boneName=spine_03}}
     {targetBone {boneName=head}}
     {targetBone {boneName=upperarm_twist_01_l}}
    ]
    constraintWeights [1,1,1,1,1,1,1,1]
   }
   n21 = "Constraint - dyn_long_hair_l_03 (PPPPPPPP)"!(
    ComponentPose=@n20
    Alpha=1
    ConstraintWeights_0=0.5705
    ConstraintWeights_1:float=0.3081
    ConstraintWeights_2:float=0.0799
    ConstraintWeights_3:float=0.0184
    ConstraintWeights_4:float=0.0098
    ConstraintWeights_5:float=0.0056
    ConstraintWeights_6:float=0.0043
    ConstraintWeights_7:float=0.0035
   ) -> (Pose) {
    alphaBoolBlend {blendOption=Linear}
    boneToModify {boneName=dyn_long_hair_l_03}
    constraintSetup [
     {targetBone {boneName=spine_05}}
     {targetBone {boneName=clavicle_l}}
     {targetBone {boneName=spine_04}}
     {targetBone {boneName=neck_01}}
     {targetBone {boneName=neck_02}}
     {targetBone {boneName=upperarm_twist_01_l}}
     {targetBone {boneName=spine_03}}
     {targetBone {boneName=deform_upperarm_fr_l}}
    ]
    constraintWeights [1,1,1,1,1,1,1,1]
   }
   n22 = "Constraint - base_long_hair_r (PPPPPPPP)"!(
    ComponentPose=@n21
    Alpha=1
    ConstraintWeights_0=0.4804
    ConstraintWeights_1:float=0.2223
    ConstraintWeights_2:float=0.1336
    ConstraintWeights_3:float=0.0755
    ConstraintWeights_4:float=0.0678
    ConstraintWeights_5:float=0.019
    ConstraintWeights_6:float=0.0008
    ConstraintWeights_7:float=0.0006
   ) -> (Pose) {
    alphaBoolBlend {blendOption=Linear}
    boneToModify {boneName=base_long_hair_r}
    constraintSetup [
     {targetBone {boneName=spine_05}}
     {targetBone {boneName=neck_02}}
     {targetBone {boneName=neck_01}}
     {targetBone {boneName=clavicle_r}}
     {targetBone {boneName=head}}
     {targetBone {boneName=spine_03}}
     {targetBone {boneName=clavicle_l}}
     {targetBone {boneName=spine_04}}
    ]
    constraintWeights [1,1,1,1,1,1,1,1]
   }
   n23 = "Constraint - dyn_long_hair_r_02 (PPPPPPPP)"!(
    ComponentPose=@n22
    Alpha=1
    ConstraintWeights_0=0.4819
    ConstraintWeights_1:float=0.4012
    ConstraintWeights_2:float=0.0409
    ConstraintWeights_3:float=0.0318
    ConstraintWeights_4:float=0.026
    ConstraintWeights_5:float=0.0098
    ConstraintWeights_6:float=0.0048
    ConstraintWeights_7:float=0.0036
   ) -> (Pose) {
    alphaBoolBlend {blendOption=Linear}
    boneToModify {boneName=dyn_long_hair_r_02}
    constraintSetup [
     {targetBone {boneName=spine_05}}
     {targetBone {boneName=clavicle_r}}
     {targetBone {boneName=neck_01}}
     {targetBone {boneName=neck_02}}
     {targetBone {boneName=spine_04}}
     {targetBone {boneName=spine_03}}
     {targetBone {boneName=head}}
     {targetBone {boneName=upperarm_twist_01_r}}
    ]
    constraintWeights [1,1,1,1,1,1,1,1]
   }
   n24 = "Constraint - dyn_long_hair_r_03 (PPPPPPPP)"!(
    ComponentPose=@n23
    Alpha=1
    ConstraintWeights_0=0.5654
    ConstraintWeights_1:float=0.326
    ConstraintWeights_2:float=0.0821
    ConstraintWeights_3:float=0.0102
    ConstraintWeights_4:float=0.0057
    ConstraintWeights_5:float=0.0045
    ConstraintWeights_6:float=0.0037
    ConstraintWeights_7:float=0.0024
   ) -> (Pose) {
    alphaBoolBlend {blendOption=Linear}
    boneToModify {boneName=dyn_long_hair_r_03}
    constraintSetup [
     {targetBone {boneName=spine_05}}
     {targetBone {boneName=clavicle_r}}
     {targetBone {boneName=spine_04}}
     {targetBone {boneName=neck_01}}
     {targetBone {boneName=upperarm_twist_01_r}}
     {targetBone {boneName=neck_02}}
     {targetBone {boneName=deform_upperarm_fr_r}}
     {targetBone {boneName=spine_03}}
    ]
    constraintWeights [1,1,1,1,1,1,1,1]
   }
   Save cached pose 'HairCache'(Pose=Component To Local(ComponentPose=@n24)) {cacheName=HairCache}
   n25 = "Constraint - master_medallion (PPPPPP)"!(
    ComponentPose=Local To Component(LocalPose=Use cached pose 'HairCache'())
    Alpha=1
    ConstraintWeights_0=0.4148
    ConstraintWeights_1:float=0.3214
    ConstraintWeights_2:float=0.2366
    ConstraintWeights_3:float=0.0158
    ConstraintWeights_4:float=0.006
    ConstraintWeights_5:float=0.0053
   ) -> (Pose) {
    alphaBoolBlend {blendOption=Linear}
    boneToModify {boneName=master_medallion}
    constraintSetup [
     {targetBone {boneName=neck_01}}
     {targetBone {boneName=spine_05}}
     {targetBone {boneName=neck_02}}
     {targetBone {boneName=head}}
     {targetBone {boneName=clavicle_r}}
     {targetBone {boneName=clavicle_l}}
    ]
    constraintWeights [1,1,1,1,1,1]
   }
   n26 = Anim Dynamics - Bone: dyn_medallion(ComponentPose=@n25, Alpha=1, ExternalForce="0,0,0") {
    angularDampingOverride=1
    angularSpringConstant=20000
    bAngularSpring=true
    bOverrideAngularDamping=true
    bUseGravityOverride=true
    gravityScale=0.35
    simulationSpace=RootRelative
    alphaBoolBlend {blendOption=Linear}
    boundBone {boneName=dyn_medallion}
    componentLinearAccScale=<0.6 0.6 0.6>
    componentLinearVelScale=<0.6 0.6 0.6>
    gravityOverride {z=-980}
    physicsBodyDefinitions [{
      boundBone {boneName=dyn_medallion}
      boxExtents=<20 30 30>
      constraintSetup {angularLimitsMax=<15 15 3>, angularLimitsMin=<-15 -15 -15>, angularTarget=<1 0 0>}
      localJointOffset=<0.2 0 0>
     }]
   }
   n27 = Get bIsInFrontEnd()
   n28 = Anim Dynamics - Bone: dyn_medallion!(ComponentPose=@n26, bAlphaBoolEnabled:bool=@n27, ExternalForce="0,0,0") -> (Pose) {
    alphaInputType=Bool
    angularDampingOverride=1
    angularSpringConstant=20000
    bAngularSpring=true
    bOverrideAngularDamping=true
    bUseGravityOverride=true
    gravityScale=0.35
    simulationSpace=World
    alphaBoolBlend {blendInTime=0.25, blendOption=Linear, blendOutTime=0.25}
    boundBone {boneName=dyn_medallion}
    componentLinearAccScale=<0.6 0.6 0.6>
    componentLinearVelScale=<0.6 0.6 0.6>
    gravityOverride {z=-980}
    physicsBodyDefinitions [{
      boundBone {boneName=dyn_medallion}
      boxExtents=<20 30 30>
      constraintSetup {angularLimitsMax=<15 15 3>, angularLimitsMin=<-15 -15 -15>, angularTarget=<1 0 0>}
      localJointOffset=<0.2 0 0>
     }]
   }
   n29 = "Transform (Modify) Bone - Bone: base_medallion"!(ComponentPose=@n28, bAlphaBoolEnabled:bool=@n27, Translation="0,0,0", Rotation="0,12,0", Scale="1,1,1") -> (Pose) {
    alphaInputType=Bool
    rotationMode=BMM_Additive
    rotationSpace=BCS_BoneSpace
    alphaBoolBlend {blendInTime=0.25, blendOption=Linear, blendOutTime=0.25}
    boneToModify {boneName=base_medallion}
    rotation {yaw=12}
   }
   Save cached pose 'MedallionCache'(Pose=Component To Local(ComponentPose=@n29)) {cacheName=MedallionCache}
  ]}
 EventGraph {nodes [
   Event BlueprintUpdateAnimation()
   > n1 = Gravity Override State Function DEPRECATED(
    CharacterPartAnimInstance=Self-Reference()
    GameTimeInSeconds=Get Game Time in Seconds() -> (Game Time in Seconds:double)
    GravityOverrideParams=Get Gravity Override Params()
   )
   > Set Gravity Override(Gravity Override=@n1)
   Event OnParachuteOpened()
   > Switch on EGliderType(Selection=Get PlayerGliderType() -> (PlayerGliderType=HangGlider))
   > Set HangGliding(HangGliding:bool=true) -> (Output_Get:bool)
   Event OnParachuteClosed()
   > Set HangGliding(HangGliding=false)
  ]}
]
variables [
 Game Time in Seconds {type=number}
 Gravity Override {type=Vector}
 Gravity Override Params {type=GravityOverrideParamsStruct}
 HangGliding {type=boolean}
]
